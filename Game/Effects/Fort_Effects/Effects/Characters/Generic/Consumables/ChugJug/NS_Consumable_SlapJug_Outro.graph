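# Niagara effect for the Slap Jug consumable outro that tosses a short-lived jug mesh, presumably the discarded container.
# A single burst emitter sets the mesh index from User.ColorVariant, orients and offsets it, then launches it with velocity.
# Gravity and local acceleration move it while curves drive rotation rate and shrink it at end of life.
# A platform-set check gates bouncing collision by quality level.
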NiagaraSystem /Game/Effects/Fort_Effects/Effects/Characters/Generic/Consumables/ChugJug/NS_Consumable_SlapJug_Outro
dependencies {
 usedDataInterfaces [
  &NiagaraDataInterfaceMeshRendererInfo
  &NiagaraDataInterfaceCurve
  &NiagaraDataInterfacePlatformSet
 ]
 usedDynamicInputs [&RandomRangeVector]
 usedModules [&SystemState]
}
emitters {ChugJug_Mesh {
  emitterUpdateScript [
   &EmitterState(Life Cycle Mode=::NewEnumerator1, Loop Behavior=::NewEnumerator1, Loop Duration=1, UseLoopDelay=false)
   &SpawnBurst_Instantaneous(Use Spawn Probability=false, UseLoopCountLimit=false)
  ]
  particleSpawnScript [
   /SetVariables(Particles.MeshIndex=$User.ColorVariant) {isSetParameters=true}
   &InitializeParticle(
    Color Mode=::NewEnumerator0
    Lifetime=0.6
    Mass=200
    Mass Mode=::NewEnumerator1
    Mesh Scale Mode=::NewEnumerator3
    Mesh Uniform Scale=0.8
    Position Offset=<-10 0 0>
    Position Offset Coordinate Space=::Local
    UsePositionOffset=true
    Write Material Random=true
   )
   &InitialMeshOrientation(Mesh Axis To Orient=<0 -1 0>, Mesh Orientation Mode=::NewEnumerator2, Rotation=<0.2076 0.6224 0.4888>, Use Rotation Vector=true)
   &AddVelocity(Velocity=<-0.5 1 -1.5>, Velocity Speed Scale=500)
  ]
  particleUpdateScript [
   &ParticleState()
   &GravityForce()
   &UpdateMeshOrientation(
    Mesh Renderer=DataInterface() {propertyValues {MeshRenderer=&self:ChugJug_Mesh_0.NiagaraMeshRendererProperties_0}}
    Orientation Method=::NewEnumerator6
    Particle Radius Scale=3
    Rotation Rate=&FloatFromCurve(FloatCurve=DataInterface() {propertyValues {Curve {keys [
         {arriveTangent=0.002, interpMode=RCIM_Cubic, leaveTangent=0.002, tangentMode=RCTM_User, time=0, value=0}
         {arriveTangent=21.4184, interpMode=RCIM_Cubic, leaveTangent=1.1112, tangentMode=RCTM_Break, time=0.1408, value=0.9847}
         {arriveTangent=-0.0405, interpMode=RCIM_Cubic, leaveTangent=-0.0405, tangentMode=RCTM_User, time=1, value=0}
        ]}}})
   )
   &ScaleMeshSize(Scale Factor=&VectorFromFloat(Value=&FloatFromCurve(FloatCurve=DataInterface() {propertyValues {Curve {keys [
          <at 0.0008 1 RCIM_Cubic>
          <at 0.1892 1 RCIM_Cubic>
          {arriveTangent=0.0519, interpMode=RCIM_Cubic, leaveTangent=0.0519, tangentMode=RCTM_User, time=0.7, value=1}
          {arriveTangent=-7.9973, tangentMode=RCTM_User, time=1, value=0}
         ]}}})))
   &AccelerationForce(Acceleration=<-1000 2000 0>, Coordinate Space=::Local)
   &Collision(
    Age Colliding Particles=false
    Collision Enabled=&IsPlatformSetActive(New Platform Set=DataInterface() {propertyValues {Platforms {qualityLevelMask=-4}}})
    Friction=2
    KillOnCollision=false
    Module.Control Rest State Via Surface Normal=false
    Module.EnableMaxCollisionCount=false
    Module.Randomize Collision Normal Vector=false
    Restitution=0.25
   )
   &SolveForcesAndVelocity(Clamp Velocity=false, Limit Acceleration=false)
  ]
  rendererClasses [&NiagaraMeshRendererProperties]
  renderers [{rendererClass=&NiagaraMeshRendererProperties, rendererIndex=0}]
 }}
summary {
 systemName=NS_Consumable_SlapJug_Outro
 emitters [{emitterName=ChugJug_Mesh, rendererClasses [&NiagaraMeshRendererProperties]}]
 userVariables [User.ColorVariant {type=NiagaraInt32}]
}
